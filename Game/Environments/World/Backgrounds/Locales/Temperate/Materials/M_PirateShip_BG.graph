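# Nanite-enabled lit material for a background pirate ship that sails a slow elliptical loop while rocking.
# World-position math combines two sine-driven rotations, phase-shifted by object X position, with an orbit around the object pivot, plus a normal fix.
# Vertex color red masks an added sine wave displacement.
# An emissive mask glows according to the global time-of-day parameter, fading out toward mid-cycle.
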Material /Game/Environments/World/Backgrounds/Locales/Temperate/Materials/M_PirateShip_BG
graph {
 properties {bUsedWithNanite=true, pixelDepthOffsetMode=PDOM_Legacy, usedShadingModels=Default Lit}
 nodes [
  Add()
  n1 = Constant3Vector() {constant=<rgba 1 0 0 0>}
  n2 = ObjectPositionWS()
  n3 = Add(A=Time(), B=Frac(None=ComponentMask(None=@n2) {r=true}))
  n4 = ScalarParameter() {defaultValue=0.15, parameterName=SpeedYaw}
  n5 = Multiply(A=Sine(None=Multiply(A=@n3, B=@n4)), B=ScalarParameter() {defaultValue=0.008, parameterName=Yaw})
  n6 = ObjectPositionWS()
  n7 = WorldPosition()
  n8 = Constant3Vector() {constant=<rgba 0 1 0 0>}
  n9 = ScalarParameter() {defaultValue=0.2, parameterName=SpeedRoll}
  n10 = Multiply(A=Sine(None=Multiply(A=@n3, B=@n9)), B=ScalarParameter() {defaultValue=0.008, parameterName=Roll})
  n11 = ObjectPositionWS()
  n12 = WorldPosition()
  n13 = Transform(None=Constant3Vector() {constant=<rgba 1 0.1 0 0>}) {transformSourceType=TRANSFORMSOURCE_Local}
  n14 = VectorParameter() {parameterName=Elipsis, defaultValue=<rgba 2 1 1 1>}
  n15 = Add(
   A=Add(
    A=RotateAboutAxis(NormalizedRotationAxis=@n1, RotationAngle=@n5, PivotPoint=@n6, Position=@n7.XYZ)
    B=RotateAboutAxis(NormalizedRotationAxis=@n8, RotationAngle=@n10, PivotPoint=@n11, Position=@n12.XYZ)
   )
   B=Multiply(A=Multiply(A=@n13, B=@n14.RGB), B=ScalarParameter() {defaultValue=58666.8125, parameterName=MovementRadius})
  )
  n16 = Constant3Vector() {constant=<rgba 0 0 1 0>}
  n17 = WorldPosition()
  n18 = Frac(None=Multiply(A=Time(), B=ScalarParameter() {defaultValue=-0.002, parameterName=MovementSpeed}))
  n19 = Add(A=@n15, B=RotateAboutAxis(NormalizedRotationAxis=@n16, RotationAngle=@n18, PivotPoint=ObjectPositionWS(), Position=Add(A=@n15, B=@n17.XYZ)))
  n20 = Add(A=WorldPosition().XYZ, B=@n19)
  n21 = Sine(None=Divide(A=ComponentMask(None=@n20) {b=true, g=true, r=true}, B=ScalarParameter() {defaultValue=1903.2775, parameterName=WaveLength}))
  n22 = ScalarParameter() {defaultValue=194.1573, parameterName=WaveAmp}
  Add(A=Multiply(A=VertexColor().R, B=Multiply(A=@n21, B=@n22)), B=@n19)
  Constant3Vector() {constant=<rgba 0 0 1 0>}
  MaterialFunctionCall(Rotation Axis=@n16, Rotation Angle=@n18) {materialFunction=&FixRotateAboutAxisNormals}
  n23 = ScalarParameter() {defaultValue=3.4615, parameterName=night}
  n24 = ScalarParameter() {defaultValue=-11.9329, parameterName=Day}
  n25 = CollectionParameter() {collection=&FortniteMaterialParameters, parameterName=TimeOfDay_0-1}
  n26 = Multiply(A=@n25, B=OneMinus(None=@n25))
  Multiply(
   A=Clamp(None=LinearInterpolate(A=@n23, B=@n24, Alpha=@n26))
   B=Multiply(
    A=VectorParameter() {parameterName=Emissive, defaultValue=<rgba 10 4.2179 1.0629 1>}.RGB
    B=TextureSample() {texture=/Game/Environments/World/Backgrounds/Locales/Temperate/Textures/T_PirateShip_BG_E}.R
   )
  )
  ScalarParameter() {defaultValue=0.5, parameterName=S}
  ScalarParameter() {defaultValue=0.7861, parameterName=R}
  TextureSample() {texture=/Game/Environments/World/Backgrounds/Locales/Temperate/Textures/T_PirateShip_BG}
 ]
}
